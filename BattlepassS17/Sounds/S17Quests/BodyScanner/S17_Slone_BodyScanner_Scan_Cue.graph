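# Sound cue for the Season 17 Slone body scanner quest, playing the scanner's scan sound.
# A single wave player feeds the S17_Slone_BodyScanner_Scan wave straight to the output, with no randomization or modulation.
# It uses the shared Season 17 quest attenuation asset and plays through the Fort_Quests sound class.
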SoundCue /BattlepassS17/Sounds/S17Quests/BodyScanner/S17_Slone_BodyScanner_Scan_Cue
root=WavePlayer_0
nodes [{id=WavePlayer_0, properties {soundWaveAssetPtr=&/S17_Slone_BodyScanner_Scan}}]
properties {
 attenuationSettings=/BattlepassS17/Sounds/S17_Quest_Att
 maxDistance=4000
 soundClassObject=&Fort_Quests
}
